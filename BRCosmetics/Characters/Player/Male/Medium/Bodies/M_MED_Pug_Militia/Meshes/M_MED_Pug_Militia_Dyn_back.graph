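# Physics asset for the back dynamic clip bones of the M_MED_Pug_Militia male character body, likely for secondary motion on dangling back pieces.
# Capsule bodies form two chains from small master-bone capsules: one segment on the left, three on the right.
# Each constraint joins a segment to its parent with twist and swing1 locked, leaving swing2 as the only unlocked angular axis, so each joint hinges in one plane.
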PhysicsAsset /BRCosmetics/Characters/Player/Male/Medium/Bodies/M_MED_Pug_Militia/Meshes/M_MED_Pug_Militia_Dyn_back
bodies {
 dyn_clip_bk_l_01 [{shapeType=Capsule, length=3.6875, radius=3, center=<3 0 0>, rotation=<rot 0 0 90>}]
 dyn_clip_bk_l_MSTR [{shapeType=Capsule, length=0.5, radius=0.5}]
 dyn_clip_bk_r_01 [{shapeType=Capsule, length=3.6875, radius=3, center=<3 0 0>, rotation=<rot 0 0 90>}]
 dyn_clip_bk_r_02 [{shapeType=Capsule, length=3.6875, radius=3, center=<3 0 0>, rotation=<rot 0 0 90>}]
 dyn_clip_bk_r_03 [{shapeType=Capsule, length=3.6875, radius=3, center=<3 0 0>, rotation=<rot 0 0 90>}]
 dyn_clip_bk_r_MSTR [{shapeType=Capsule, length=0.5, radius=0.5}]
}
constraints [
 {
  bone1Name=dyn_clip_bk_l_01
  bone2Name=dyn_clip_bk_l_MSTR
  swing1LimitDegrees=45
  swing1Motion=Locked
  swing2LimitDegrees=45
  twistLimitDegrees=45
  twistMotion=Locked
 }
 {
  bone1Name=dyn_clip_bk_r_01
  bone2Name=dyn_clip_bk_r_MSTR
  swing1LimitDegrees=45
  swing1Motion=Locked
  swing2LimitDegrees=45
  twistLimitDegrees=45
  twistMotion=Locked
 }
 {
  bone1Name=dyn_clip_bk_r_02
  bone2Name=dyn_clip_bk_r_01
  swing1LimitDegrees=45
  swing1Motion=Locked
  swing2LimitDegrees=45
  twistLimitDegrees=45
  twistMotion=Locked
 }
 {
  bone1Name=dyn_clip_bk_r_03
  bone2Name=dyn_clip_bk_r_02
  swing1LimitDegrees=45
  swing1Motion=Locked
  swing2LimitDegrees=45
  twistLimitDegrees=45
  twistMotion=Locked
 }
]
