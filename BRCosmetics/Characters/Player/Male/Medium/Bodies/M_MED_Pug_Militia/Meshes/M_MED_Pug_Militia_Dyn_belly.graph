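# Physics asset for the dynamic belly shirt bones on the medium male Pug Militia player body cosmetic.
# Three small, identical capsules cover the dyn_belly_shirt bones, and a larger offset capsule covers spine_01.
# Each shirt bone is constrained directly to spine_01 rather than chained, with tight swing and twist limits that allow only slight angular movement relative to the torso.
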PhysicsAsset /BRCosmetics/Characters/Player/Male/Medium/Bodies/M_MED_Pug_Militia/Meshes/M_MED_Pug_Militia_Dyn_belly
bodies {
 dyn_belly_shirt_01 [{shapeType=Capsule, length=5, radius=3, center=<2 0 0>, rotation=<rot 90 0 0>}]
 dyn_belly_shirt_02 [{shapeType=Capsule, length=5, radius=3, center=<2 0 0>, rotation=<rot 90 0 0>}]
 dyn_belly_shirt_03 [{shapeType=Capsule, length=5, radius=3, center=<2 0 0>, rotation=<rot 90 0 0>}]
 spine_01 [{shapeType=Capsule, length=7.3438, radius=5.2812, center=<-4.0656 0.2928 -2.299>}]
}
constraints [
 {bone1Name=dyn_belly_shirt_03, bone2Name=spine_01, swing1LimitDegrees=5, swing2LimitDegrees=5, twistLimitDegrees=5}
 {bone1Name=dyn_belly_shirt_02, bone2Name=spine_01, swing1LimitDegrees=5, swing2LimitDegrees=5, twistLimitDegrees=5}
 {bone1Name=dyn_belly_shirt_01, bone2Name=spine_01, swing1LimitDegrees=5, swing2LimitDegrees=5, twistLimitDegrees=5}
]
