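# Physics asset for the chest dynamics of the Pug Militia medium male body, covering a front-left clip, two grenades and four torso pouches.
# Each item pairs a tiny capsule on its MSTR bone with a larger capsule on its 01 bone, joined by one constraint.
# Grenade and pouch constraints carry tight 5-degree swing and twist limits.
# The clip constraint locks swing1 and twist, leaving swing2 as its only rotational axis.
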PhysicsAsset /BRCosmetics/Characters/Player/Male/Medium/Bodies/M_MED_Pug_Militia/Meshes/M_MED_Pug_Militia_Dyn_chest
bodies {
 dyn_clip_fr_l_01 [{shapeType=Capsule, length=3.6875, radius=3, center=<3 0 0>, rotation=<rot 0 0 90>}]
 dyn_clip_fr_l_MSTR [{shapeType=Capsule, length=0.5, radius=0.5}]
 dyn_grenade_in_01 [{shapeType=Capsule, length=5, radius=3, center=<1 0 0>, rotation=<rot 90 0 0>}]
 dyn_grenade_in_MSTR [{shapeType=Capsule, length=0.5, radius=0.5}]
 dyn_grenade_out_01 [{shapeType=Capsule, length=5, radius=3, center=<1 0 0>, rotation=<rot 90 0 0>}]
 dyn_grenade_out_MSTR [{shapeType=Capsule, length=0.5, radius=0.5}]
 dyn_torso_pouch_bott_l_01 [{shapeType=Capsule, length=5, radius=3, center=<1 0 0>, rotation=<rot 90 0 0>}]
 dyn_torso_pouch_bott_l_MSTR [{shapeType=Capsule, length=0.5, radius=0.5}]
 dyn_torso_pouch_in_r_01 [{shapeType=Capsule, length=5, radius=3, center=<1 0 0>, rotation=<rot 90 0 0>}]
 dyn_torso_pouch_in_r_MSTR [{shapeType=Capsule, length=0.5, radius=0.5}]
 dyn_torso_pouch_out_r_01 [{shapeType=Capsule, length=5, radius=3, center=<1 0 0>, rotation=<rot 90 0 0>}]
 dyn_torso_pouch_out_r_MSTR [{shapeType=Capsule, length=0.5, radius=0.5}]
 dyn_torso_pouch_top_l_01 [{shapeType=Capsule, length=5, radius=3, center=<1 0 0>, rotation=<rot 90 0 0>}]
 dyn_torso_pouch_top_l_MSTR [{shapeType=Capsule, length=0.5, radius=0.5}]
}
constraints [
 {
  bone1Name=dyn_clip_fr_l_01
  bone2Name=dyn_clip_fr_l_MSTR
  swing1LimitDegrees=45
  swing1Motion=Locked
  swing2LimitDegrees=45
  twistLimitDegrees=45
  twistMotion=Locked
 }
 {bone1Name=dyn_grenade_in_01, bone2Name=dyn_grenade_in_MSTR, swing1LimitDegrees=5, swing2LimitDegrees=5, twistLimitDegrees=5}
 {bone1Name=dyn_grenade_out_01, bone2Name=dyn_grenade_out_MSTR, swing1LimitDegrees=5, swing2LimitDegrees=5, twistLimitDegrees=5}
 {bone1Name=dyn_torso_pouch_top_l_01, bone2Name=dyn_torso_pouch_top_l_MSTR, swing1LimitDegrees=5, swing2LimitDegrees=5, twistLimitDegrees=5}
 {bone1Name=dyn_torso_pouch_bott_l_01, bone2Name=dyn_torso_pouch_bott_l_MSTR, swing1LimitDegrees=5, swing2LimitDegrees=5, twistLimitDegrees=5}
 {bone1Name=dyn_torso_pouch_in_r_01, bone2Name=dyn_torso_pouch_in_r_MSTR, swing1LimitDegrees=5, swing2LimitDegrees=5, twistLimitDegrees=5}
 {bone1Name=dyn_torso_pouch_out_r_01, bone2Name=dyn_torso_pouch_out_r_MSTR, swing1LimitDegrees=5, swing2LimitDegrees=5, twistLimitDegrees=5}
]
