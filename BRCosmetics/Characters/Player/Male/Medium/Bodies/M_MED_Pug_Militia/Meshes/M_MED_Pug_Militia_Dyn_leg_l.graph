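# Physics asset for the Pug Militia male character's left leg, simulating dangling shoe laces and a shoe band.
# Capsule bodies on the lower calf and foot serve as anchors and collision shapes for small capsules on the dynamic bones.
# Inner and outer two-segment lace chains hang from master bodies on the calf, and a one-segment band from a master on the foot.
# Masters rotate freely against their parent, while child segments have limited swing and tight twist.
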PhysicsAsset /BRCosmetics/Characters/Player/Male/Medium/Bodies/M_MED_Pug_Militia/Meshes/M_MED_Pug_Militia_Dyn_leg_l
bodies {
 calf_twist_01_l [{shapeType=Capsule, length=2.0938, radius=2.9062, center=<-3.212 0.4286 -0.4175>, rotation=<rot 84.9999 0 0>}]
 dyn_shoe_band_l_01 [{shapeType=Capsule, length=2.0938, radius=0.5, center=<0.05 0 0>, rotation=<rot 90 0 0>}]
 dyn_shoe_band_l_MSTR [{shapeType=Capsule, length=0.5, radius=0.5}]
 dyn_shoe_lace_l_in_01 [{shapeType=Capsule, length=2.0938, radius=0.5, center=<1.1 0 0>, rotation=<rot 90 0 0>}]
 dyn_shoe_lace_l_in_02 [{shapeType=Capsule, length=2.0938, radius=0.5, center=<1.1 0 0>, rotation=<rot 90 0 0>}]
 dyn_shoe_lace_l_in_MSTR [{shapeType=Capsule, length=0.5, radius=0.5}]
 dyn_shoe_lace_l_out_01 [{shapeType=Capsule, length=2.0938, radius=0.5, center=<1.1 0 0>, rotation=<rot 90 0 0>}]
 dyn_shoe_lace_l_out_02 [{shapeType=Capsule, length=2.0938, radius=0.5, center=<1.1 0 0>, rotation=<rot 90 0 0>}]
 dyn_shoe_lace_l_out_MSTR1 [{shapeType=Capsule, length=0.5, radius=0.5}]
 foot_l [{shapeType=Capsule, length=4.625, radius=2.9688, center=<-1.6475 -0.5337 -0.5879>, rotation=<rot 90 0 0>}]
}
constraints [
 {bone1Name=dyn_shoe_lace_l_in_01, bone2Name=dyn_shoe_lace_l_in_MSTR, swing1LimitDegrees=60, swing2LimitDegrees=60, twistLimitDegrees=5}
 {bone1Name=dyn_shoe_lace_l_in_02, bone2Name=dyn_shoe_lace_l_in_01, swing1LimitDegrees=60, swing2LimitDegrees=60, twistLimitDegrees=5}
 {bone1Name=dyn_shoe_lace_l_out_01, bone2Name=dyn_shoe_lace_l_out_MSTR1, swing1LimitDegrees=60, swing2LimitDegrees=60, twistLimitDegrees=5}
 {bone1Name=dyn_shoe_lace_l_out_02, bone2Name=dyn_shoe_lace_l_out_01, swing1LimitDegrees=60, swing2LimitDegrees=60, twistLimitDegrees=5}
 {
  bone1Name=dyn_shoe_lace_l_in_MSTR
  bone2Name=calf_twist_01_l
  swing1LimitDegrees=45
  swing1Motion=Free
  swing2LimitDegrees=45
  swing2Motion=Free
  twistLimitDegrees=45
  twistMotion=Free
 }
 {
  bone1Name=dyn_shoe_lace_l_out_MSTR1
  bone2Name=calf_twist_01_l
  swing1LimitDegrees=45
  swing1Motion=Free
  swing2LimitDegrees=45
  swing2Motion=Free
  twistLimitDegrees=45
  twistMotion=Free
 }
 {
  bone1Name=dyn_shoe_band_l_MSTR
  bone2Name=foot_l
  swing1LimitDegrees=45
  swing1Motion=Free
  swing2LimitDegrees=45
  swing2Motion=Free
  twistLimitDegrees=45
  twistMotion=Free
 }
 {bone1Name=dyn_shoe_band_l_01, bone2Name=dyn_shoe_band_l_MSTR, swing1LimitDegrees=25, swing2LimitDegrees=25, twistLimitDegrees=5}
]
