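# Physics asset for dynamic accessories on the right leg of the male medium Pug Militia body: shoe laces ("race" bones), a shoe band and a pouch with straps.
# Capsule bodies form short chains with limited swing and tight twist.
# Lace and band masters attach to calf and foot bodies with free rotation.
# The pouch carries a three-link strap chain, and its master has no constraint to a leg body.
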PhysicsAsset /BRCosmetics/Characters/Player/Male/Medium/Bodies/M_MED_Pug_Militia/Meshes/M_MED_Pug_Militia_Dyn_leg_r
bodies {
 calf_twist_01_r [{shapeType=Capsule, length=2.0938, radius=2.9062, center=<3.212 -0.4286 0.4175>, rotation=<rot 84.9999 0 0>}]
 dyn_pouch_01 [{shapeType=Capsule, length=5, radius=3, center=<3.5943 0 0>, rotation=<rot 90 0 0>}]
 dyn_pouch_MSTR [{shapeType=Capsule, length=0.505, radius=0.505}]
 dyn_pouch_strap_01 [{shapeType=Capsule, length=5, radius=0.7188, center=<2 0 0>, rotation=<rot 90 0 0>}]
 dyn_pouch_strap_02 [{shapeType=Capsule, length=5.7188, radius=0.7188, center=<3.1786 0 0>, rotation=<rot 90 0 0>}]
 dyn_pouch_strap_03 [{shapeType=Capsule, length=4.4062, radius=0.7188, center=<2 0 0>, rotation=<rot 90 0 0>}]
 dyn_shoe_band_r_01 [{shapeType=Capsule, length=2.0938, radius=0.5, center=<0.05 0 0>, rotation=<rot 90 0 0>}]
 dyn_shoe_band_r_MSTR [{shapeType=Capsule, length=0.5, radius=0.5}]
 dyn_shoe_race_r_in_01 [{shapeType=Capsule, length=2.0938, radius=0.5, center=<1.1 0 0>, rotation=<rot 90 0 0>}]
 dyn_shoe_race_r_in_02 [{shapeType=Capsule, length=2.0938, radius=0.5, center=<1.1 0 0>, rotation=<rot 90 0 0>}]
 dyn_shoe_race_r_in_MSTR [{shapeType=Capsule, length=0.5, radius=0.5}]
 dyn_shoe_race_r_out_01 [{shapeType=Capsule, length=2.0938, radius=0.5, center=<1.1 0 0>, rotation=<rot 90 0 0>}]
 dyn_shoe_race_r_out_02 [{shapeType=Capsule, length=2.0938, radius=0.5, center=<1.1 0 0>, rotation=<rot 90 0 0>}]
 dyn_shoe_race_r_out_MSTR1 [{shapeType=Capsule, length=0.5, radius=0.5}]
 foot_r [{shapeType=Capsule, length=4.625, radius=2.9688, center=<1.6475 0.5337 0.5879>, rotation=<rot 90 0 0>}]
}
constraints [
 {bone1Name=dyn_shoe_race_r_in_01, bone2Name=dyn_shoe_race_r_in_MSTR, swing1LimitDegrees=60, swing2LimitDegrees=60, twistLimitDegrees=5}
 {bone1Name=dyn_shoe_race_r_in_02, bone2Name=dyn_shoe_race_r_in_01, swing1LimitDegrees=60, swing2LimitDegrees=60, twistLimitDegrees=5}
 {bone1Name=dyn_shoe_race_r_out_01, bone2Name=dyn_shoe_race_r_out_MSTR1, swing1LimitDegrees=60, swing2LimitDegrees=60, twistLimitDegrees=5}
 {bone1Name=dyn_shoe_race_r_out_02, bone2Name=dyn_shoe_race_r_out_01, swing1LimitDegrees=60, swing2LimitDegrees=60, twistLimitDegrees=5}
 {
  bone1Name=dyn_shoe_race_r_in_MSTR
  bone2Name=calf_twist_01_r
  swing1LimitDegrees=45
  swing1Motion=Free
  swing2LimitDegrees=45
  swing2Motion=Free
  twistLimitDegrees=45
  twistMotion=Free
 }
 {
  bone1Name=dyn_shoe_race_r_out_MSTR1
  bone2Name=calf_twist_01_r
  swing1LimitDegrees=45
  swing1Motion=Free
  swing2LimitDegrees=45
  swing2Motion=Free
  twistLimitDegrees=45
  twistMotion=Free
 }
 {
  bone1Name=dyn_shoe_band_r_MSTR
  bone2Name=foot_r
  swing1LimitDegrees=45
  swing1Motion=Free
  swing2LimitDegrees=45
  swing2Motion=Free
  twistLimitDegrees=45
  twistMotion=Free
 }
 {bone1Name=dyn_shoe_band_r_01, bone2Name=dyn_shoe_band_r_MSTR, swing1LimitDegrees=25, swing2LimitDegrees=25, twistLimitDegrees=5}
 {bone1Name=dyn_pouch_01, bone2Name=dyn_pouch_MSTR, swing1LimitDegrees=5, swing2LimitDegrees=5, twistLimitDegrees=5}
 {bone1Name=dyn_pouch_strap_01, bone2Name=dyn_pouch_01, swing1LimitDegrees=45, swing2LimitDegrees=45, twistLimitDegrees=5}
 {bone1Name=dyn_pouch_strap_02, bone2Name=dyn_pouch_strap_01, swing1LimitDegrees=45, swing2LimitDegrees=45, twistLimitDegrees=5}
 {bone1Name=dyn_pouch_strap_03, bone2Name=dyn_pouch_strap_02, swing1LimitDegrees=45, swing2LimitDegrees=45, twistLimitDegrees=5}
]
