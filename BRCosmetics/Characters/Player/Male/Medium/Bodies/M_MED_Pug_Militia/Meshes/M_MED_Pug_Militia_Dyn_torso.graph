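# Physics asset for the dynamic torso accessories of the M_MED_Pug_Militia player body cosmetic: belt straps, clips and a pouch.
# Large pelvis, spine_01 and right thigh capsules cover the body, with small MSTR bodies anchoring accessory chains to pelvis and thigh.
# Pouch and strap joints have tight angular limits, belt strap links lock twist, and clip links swing on one axis only.
# Anchor and body joints leave all angular motion free.
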PhysicsAsset /BRCosmetics/Characters/Player/Male/Medium/Bodies/M_MED_Pug_Militia/Meshes/M_MED_Pug_Militia_Dyn_torso
bodies {
 dyn_belt_buckle_MSTR [{shapeType=Capsule, length=0.5, radius=0.5}]
 dyn_belt_clip_01 [{shapeType=Capsule, length=2.0938, radius=1.4375, center=<0.0271 0 -0.0002>, rotation=<rot 0 0 90>}]
 dyn_belt_clip_02 [{shapeType=Capsule, length=2.0938, radius=1.4375, center=<3 0 0>, rotation=<rot 0 0 90>}]
 dyn_belt_strap_01 [{shapeType=Capsule, length=2.4062, radius=1.375, center=<1 0 0>, rotation=<rot 0 0 90>}]
 dyn_belt_strap_02 [{shapeType=Capsule, length=2.4062, radius=1.375, center=<1 0 0>, rotation=<rot 0 0 90>}]
 dyn_belt_strap_MSTR [{shapeType=Capsule, length=0.5, radius=0.5}]
 dyn_pouch_01 [{shapeType=Capsule, length=5, radius=3, center=<1 0 0>, rotation=<rot 90 0 0>}]
 dyn_pouch_MSTR [{shapeType=Capsule, length=0.5, radius=0.5}]
 dyn_pouch_strap_01 [{shapeType=Capsule, length=5, radius=0.7188, center=<2 0 0>, rotation=<rot 90 0 0>}]
 dyn_pouch_strap_02 [{shapeType=Capsule, length=5.7188, radius=0.7188, center=<3.1786 0 0>, rotation=<rot 90 0 0>}]
 dyn_pouch_strap_03 [{shapeType=Capsule, length=4.4062, radius=0.7188, center=<2 0 0>, rotation=<rot 90 0 0>}]
 pelvis [{shapeType=Capsule, length=11.8125, radius=12.4688, center=<-12.4464 -1.3793 4.3749>, rotation=<rot 0 0 -10.0006>}]
 spine_01 [{shapeType=Capsule, length=7.3438, radius=5.2812, center=<-4.0656 0.2928 -2.299>}]
 thigh_r [{shapeType=Capsule, length=16.2188, radius=8.0625, center=<16.7448 -1.8762 -0.0001>, rotation=<rot 90 0 0>}]
}
constraints [
 {bone1Name=dyn_pouch_01, bone2Name=dyn_pouch_MSTR, swing1LimitDegrees=5, swing2LimitDegrees=5, twistLimitDegrees=5}
 {bone1Name=dyn_pouch_strap_01, bone2Name=dyn_pouch_01, swing1LimitDegrees=15, swing2LimitDegrees=15, twistLimitDegrees=5}
 {bone1Name=dyn_pouch_strap_02, bone2Name=dyn_pouch_strap_01, swing1LimitDegrees=15, swing2LimitDegrees=15, twistLimitDegrees=5}
 {bone1Name=dyn_pouch_strap_03, bone2Name=dyn_pouch_strap_02, swing1LimitDegrees=15, swing2LimitDegrees=15, twistLimitDegrees=5}
 {
  bone1Name=dyn_belt_strap_MSTR
  bone2Name=dyn_belt_buckle_MSTR
  swing1LimitDegrees=45
  swing1Motion=Free
  swing2LimitDegrees=45
  swing2Motion=Free
  twistLimitDegrees=45
  twistMotion=Free
 }
 {
  bone1Name=dyn_belt_strap_01
  bone2Name=dyn_belt_strap_MSTR
  swing1LimitDegrees=20
  swing2LimitDegrees=45
  twistLimitDegrees=45
  twistMotion=Locked
 }
 {
  bone1Name=dyn_belt_strap_02
  bone2Name=dyn_belt_strap_01
  swing1LimitDegrees=20
  swing2LimitDegrees=45
  twistLimitDegrees=45
  twistMotion=Locked
 }
 {
  bone1Name=dyn_belt_clip_01
  bone2Name=dyn_belt_buckle_MSTR
  swing1LimitDegrees=45
  swing1Motion=Locked
  swing2LimitDegrees=60
  twistLimitDegrees=45
  twistMotion=Locked
 }
 {
  bone1Name=dyn_belt_clip_02
  bone2Name=dyn_belt_clip_01
  swing1LimitDegrees=45
  swing1Motion=Locked
  swing2LimitDegrees=45
  twistLimitDegrees=45
  twistMotion=Locked
 }
 {
  bone1Name=spine_01
  bone2Name=pelvis
  swing1LimitDegrees=45
  swing1Motion=Free
  swing2LimitDegrees=45
  swing2Motion=Free
  twistLimitDegrees=45
  twistMotion=Free
 }
 {
  bone1Name=dyn_pouch_MSTR
  bone2Name=thigh_r
  swing1LimitDegrees=45
  swing1Motion=Free
  swing2LimitDegrees=45
  swing2Motion=Free
  twistLimitDegrees=45
  twistMotion=Free
 }
 {
  bone1Name=dyn_belt_buckle_MSTR
  bone2Name=pelvis
  swing1LimitDegrees=45
  swing1Motion=Free
  swing2LimitDegrees=45
  swing2Motion=Free
  twistLimitDegrees=45
  twistMotion=Free
 }
 {
  bone1Name=thigh_r
  bone2Name=pelvis
  swing1LimitDegrees=45
  swing1Motion=Free
  swing2LimitDegrees=45
  swing2Motion=Free
  twistLimitDegrees=45
  twistMotion=Free
 }
]
