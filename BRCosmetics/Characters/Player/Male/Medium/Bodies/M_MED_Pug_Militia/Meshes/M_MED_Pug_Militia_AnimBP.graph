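# Post-copy animation Blueprint for the Pug Militia male body cosmetic, adding secondary motion to its gear such as pouches, pads, grenades, belt and shoe details.
# It copies the attached parent's pose, then chains weighted multi-bone constraints to place each dyn_ master bone.
# Knee and elbow bend angles, computed each update, drive additive pad corrections.
# Per-region RigidBody simulations follow, with the back one disabled when a backbling is worn.
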AnimBlueprint /BRCosmetics/Characters/Player/Male/Medium/Bodies/M_MED_Pug_Militia/Meshes/M_MED_Pug_Militia_AnimBP
target_skeleton=&/M_MED_Pug_Militia_Skeleton
graphs [
 AnimGraph {nodes [
   n1 = "Constraint - dyn_clip_bk_l_MSTR (PPPPPP)"!(
    ComponentPose=Local To Component(LocalPose=Copy Pose From Mesh() {bUseAttachedParent=true})
    Alpha=1
    ConstraintWeights_0=0.3932
    ConstraintWeights_1:float=0.0454
    ConstraintWeights_2:float=0.0414
    ConstraintWeights_3:float=0.2819
    ConstraintWeights_4:float=0.1752
    ConstraintWeights_5:float=0.0629
   ) -> (Pose) {
    alphaBoolBlend {blendOption=Linear}
    boneToModify {boneName=dyn_clip_bk_l_MSTR}
    constraintSetup [
     {targetBone {boneName=clavicle_l}}
     {targetBone {boneName=upperarm_twist_02_l}}
     {targetBone {boneName=spine_02}}
     {targetBone {boneName=spine_04}}
     {targetBone {boneName=spine_05}}
     {targetBone {boneName=clavicle_r}}
    ]
    constraintWeights [1,1,1,1,1,1]
   }
   n2 = "Constraint - dyn_clip_bk_r_MSTR (PPPPPP)"!(
    ComponentPose=@n1
    Alpha=1
    ConstraintWeights_0=0.0577
    ConstraintWeights_1:float=0.048
    ConstraintWeights_2:float=0.039
    ConstraintWeights_3:float=0.2693
    ConstraintWeights_4:float=0.1756
    ConstraintWeights_5:float=0.4104
   ) -> (Pose) {
    alphaBoolBlend {blendOption=Linear}
    boneToModify {boneName=dyn_clip_bk_r_MSTR}
    constraintSetup [
     {targetBone {boneName=clavicle_l}}
     {targetBone {boneName=upperarm_twist_02_r}}
     {targetBone {boneName=spine_02}}
     {targetBone {boneName=spine_04}}
     {targetBone {boneName=spine_05}}
     {targetBone {boneName=clavicle_r}}
    ]
    constraintWeights [1,1,1,1,1,1]
   }
   n3 = "Constraint - dyn_clip_fr_l_MSTR (PPPPPPP)"!(
    ComponentPose=@n2
    Alpha=1
    ConstraintWeights_0=0.3598
    ConstraintWeights_1:float=0.0267
    ConstraintWeights_2:float=0.053
    ConstraintWeights_3:float=0.2518
    ConstraintWeights_4:float=0.2745
    ConstraintWeights_5:float=0.0148
    ConstraintWeights_6:float=0.0195
   ) -> (Pose) {
    alphaBoolBlend {blendOption=Linear}
    boneToModify {boneName=dyn_clip_fr_l_MSTR}
    constraintSetup [
     {targetBone {boneName=clavicle_l}}
     {targetBone {boneName=upperarm_twist_02_l}}
     {targetBone {boneName=spine_02}}
     {targetBone {boneName=spine_04}}
     {targetBone {boneName=spine_05}}
     {targetBone {boneName=upperarm_twist_01_l}}
     {targetBone {boneName=neck_02}}
    ]
    constraintWeights [1,1,1,1,1,1,1]
   }
   n4 = "Constraint - dyn_elbow_pad_r_MSTR (PPPPP)"!(
    ComponentPose=@n3
    Alpha=1
    ConstraintWeights_0=0.0248
    ConstraintWeights_1:float=0.0412
    ConstraintWeights_2:float=0.6537
    ConstraintWeights_3:float=0.1762
    ConstraintWeights_4:float=0.104
   ) -> (Pose) {
    alphaBoolBlend {blendOption=Linear}
    boneToModify {boneName=dyn_elbow_pad_r_MSTR}
    constraintSetup [
     {targetBone {boneName=lowerarm_twist_02_r}}
     {targetBone {boneName=upperarm_twist_01_r}}
     {targetBone {boneName=upperarm_r}}
     {targetBone {boneName=lowerarm_r}}
     {targetBone {boneName=upperarm_twist_02_r}}
    ]
    constraintWeights [1,1,1,1,1]
   }
   n5 = "Constraint - dyn_elbow_pad_l_MSTR (PPPPP)"!(
    ComponentPose=@n4
    Alpha=1
    ConstraintWeights_0=0.1303
    ConstraintWeights_1:float=0.1185
    ConstraintWeights_2:float=0.0103
    ConstraintWeights_3:float=0.6855
    ConstraintWeights_4:float=0.0553
   ) -> (Pose) {
    alphaBoolBlend {blendOption=Linear}
    boneToModify {boneName=dyn_elbow_pad_l_MSTR}
    constraintSetup [
     {targetBone {boneName=lowerarm_l}}
     {targetBone {boneName=upperarm_twist_02_l}}
     {targetBone {boneName=lowerarm_twist_02_l}}
     {targetBone {boneName=upperarm_l}}
     {targetBone {boneName=upperarm_twist_01_l}}
    ]
    constraintWeights [1,1,1,1,1]
   }
   n6 = "Constraint - dyn_torso_pouch_bott_l_MSTR (PPPPP)"!(
    ComponentPose=@n5
    Alpha=1
    ConstraintWeights_0=0.2446
    ConstraintWeights_1:float=0.298
    ConstraintWeights_2:float=0.0548
    ConstraintWeights_3:float=0.0428
    ConstraintWeights_4:float=0.3598
   ) -> (Pose) {
    alphaBoolBlend {blendOption=Linear}
    boneToModify {boneName=dyn_torso_pouch_bott_l_MSTR}
    constraintSetup [
     {targetBone {boneName=spine_01}}
     {targetBone {boneName=spine_02}}
     {targetBone {boneName=spine_04}}
     {targetBone {boneName=clavicle_l}}
     {targetBone {boneName=pelvis}}
    ]
    constraintWeights [1,1,1,1,1]
   }
   n7 = "Constraint - dyn_torso_pouch_top_l_MSTR (PPPPPPP)"!(
    ComponentPose=@n6
    Alpha=1
    ConstraintWeights_0=0.1732
    ConstraintWeights_1:float=0.0121
    ConstraintWeights_2:float=0.1663
    ConstraintWeights_3:float=0.3482
    ConstraintWeights_4:float=0.234
    ConstraintWeights_5:float=0.0115
    ConstraintWeights_6:float=0.0548
   ) -> (Pose) {
    alphaBoolBlend {blendOption=Linear}
    boneToModify {boneName=dyn_torso_pouch_top_l_MSTR}
    constraintSetup [
     {targetBone {boneName=clavicle_l}}
     {targetBone {boneName=upperarm_twist_02_l}}
     {targetBone {boneName=spine_01}}
     {targetBone {boneName=spine_02}}
     {targetBone {boneName=spine_04}}
     {targetBone {boneName=spine_05}}
     {targetBone {boneName=pelvis}}
    ]
    constraintWeights [1,1,1,1,1,1,1]
   }
   n8 = "Constraint - dyn_torso_pouch_in_r_MSTR (PPPPP)"!(
    ComponentPose=@n7
    Alpha=1
    ConstraintWeights_0=0.0574
    ConstraintWeights_1:float=0.1771
    ConstraintWeights_2:float=0.3788
    ConstraintWeights_3:float=0.26
    ConstraintWeights_4:float=0.1267
   ) -> (Pose) {
    alphaBoolBlend {blendOption=Linear}
    boneToModify {boneName=dyn_torso_pouch_in_r_MSTR}
    constraintSetup [
     {targetBone {boneName=clavicle_r}}
     {targetBone {boneName=spine_01}}
     {targetBone {boneName=spine_02}}
     {targetBone {boneName=spine_04}}
     {targetBone {boneName=pelvis}}
    ]
    constraintWeights [1,1,1,1,1]
   }
   n9 = "Constraint - dyn_torso_pouch_out_r_MSTR (PPPPPP)"!(
    ComponentPose=@n8
    Alpha=1
    ConstraintWeights_0=0.0118
    ConstraintWeights_1:float=0.2245
    ConstraintWeights_2:float=0.3703
    ConstraintWeights_3:float=0.1405
    ConstraintWeights_4:float=0.1446
    ConstraintWeights_5:float=0.1083
   ) -> (Pose) {
    alphaBoolBlend {blendOption=Linear}
    boneToModify {boneName=dyn_torso_pouch_out_r_MSTR}
    constraintSetup [
     {targetBone {boneName=upperarm_twist_02_r}}
     {targetBone {boneName=spine_01}}
     {targetBone {boneName=spine_02}}
     {targetBone {boneName=spine_04}}
     {targetBone {boneName=clavicle_r}}
     {targetBone {boneName=pelvis}}
    ]
    constraintWeights [1,1,1,1,1,1]
   }
   n10 = "Constraint - dyn_grenade_out_MSTR (PPPPP)"!(
    ComponentPose=@n9
    Alpha=1
    ConstraintWeights_0=0.3688
    ConstraintWeights_1:float=0.0908
    ConstraintWeights_2:float=0.2892
    ConstraintWeights_3:float=0.1898
    ConstraintWeights_4:float=0.0615
   ) -> (Pose) {
    alphaBoolBlend {blendOption=Linear}
    boneToModify {boneName=dyn_grenade_out_MSTR}
    constraintSetup [
     {targetBone {boneName=clavicle_r}}
     {targetBone {boneName=spine_02}}
     {targetBone {boneName=spine_04}}
     {targetBone {boneName=spine_05}}
     {targetBone {boneName=upperarm_twist_02_r}}
    ]
    constraintWeights [1,1,1,1,1]
   }
   n11 = "Constraint - dyn_grenade_in_MSTR (PPPPPP)"!(
    ComponentPose=@n10
    Alpha=1
    ConstraintWeights_0=0.0231
    ConstraintWeights_1:float=0.0147
    ConstraintWeights_2:float=0.1033
    ConstraintWeights_3:float=0.3919
    ConstraintWeights_4:float=0.1887
    ConstraintWeights_5:float=0.2784
   ) -> (Pose) {
    alphaBoolBlend {blendOption=Linear}
    boneToModify {boneName=dyn_grenade_in_MSTR}
    constraintSetup [
     {targetBone {boneName=clavicle_l}}
     {targetBone {boneName=spine_01}}
     {targetBone {boneName=spine_02}}
     {targetBone {boneName=spine_04}}
     {targetBone {boneName=spine_05}}
     {targetBone {boneName=clavicle_r}}
    ]
    constraintWeights [1,1,1,1,1,1]
   }
   n12 = "Constraint - dyn_hip_pouch_fr_l_MSTR (PPP)"!(ComponentPose=@n11, Alpha=1, ConstraintWeights_0=0.0234, ConstraintWeights_1:float=0.654, ConstraintWeights_2:float=0.3226) -> (Pose) {
    alphaBoolBlend {blendOption=Linear}
    boneToModify {boneName=dyn_hip_pouch_fr_l_MSTR}
    constraintSetup [{targetBone {boneName=spine_01}}, {targetBone {boneName=thigh_twist_01_l}}, {targetBone {boneName=pelvis}}]
    constraintWeights [1,1,1]
   }
   n13 = "Constraint - dyn_hip_pouch_bk_l_MSTR (PPP)"!(ComponentPose=@n12, Alpha=1, ConstraintWeights_0=0.0268, ConstraintWeights_1:float=0.4974, ConstraintWeights_2:float=0.4758) -> (Pose) {
    alphaBoolBlend {blendOption=Linear}
    boneToModify {boneName=dyn_hip_pouch_bk_l_MSTR}
    constraintSetup [{targetBone {boneName=spine_01}}, {targetBone {boneName=thigh_twist_01_l}}, {targetBone {boneName=pelvis}}]
    constraintWeights [1,1,1]
   }
   n14 = "Constraint - dyn_knee_pad_MSTR (PPP)"!(ComponentPose=@n13, Alpha=1, ConstraintWeights_0=0.5667, ConstraintWeights_1:float=0.0135, ConstraintWeights_2:float=0.4197) -> (Pose) {
    alphaBoolBlend {blendOption=Linear}
    boneToModify {boneName=dyn_knee_pad_MSTR}
    constraintSetup [{targetBone {boneName=thigh_r}}, {targetBone {boneName=calf_twist_02_r}}, {targetBone {boneName=calf_r}}]
    constraintWeights [1,1,1]
   }
   n15 = "Constraint - dyn_pouch_MSTR (PPP)"!(ComponentPose=@n14, Alpha=1, ConstraintWeights_0=0.0457, ConstraintWeights_1:float=0.0751, ConstraintWeights_2:float=0.8792) -> (Pose) {
    alphaBoolBlend {blendOption=Linear}
    boneToModify {boneName=dyn_pouch_MSTR}
    constraintSetup [{targetBone {boneName=thigh_r}}, {targetBone {boneName=pelvis}}, {targetBone {boneName=thigh_twist_01_r}}]
    constraintWeights [1,1,1]
   }
   n16 = "Constraint - dyn_belt_buckle_MSTR (PPP)"!(ComponentPose=@n15, Alpha=1, ConstraintWeights_0=0.046, ConstraintWeights_1:float=0.3416, ConstraintWeights_2:float=0.6125) -> (Pose) {
    alphaBoolBlend {blendOption=Linear}
    boneToModify {boneName=dyn_belt_buckle_MSTR}
    constraintSetup [{targetBone {boneName=spine_01}}, {targetBone {boneName=thigh_twist_01_r}}, {targetBone {boneName=pelvis}}]
    constraintWeights [1,1,1]
   }
   n17 = "Constraint - dyn_belt_strap_MSTR (PPPP)"!(
    ComponentPose=@n16
    Alpha=1
    ConstraintWeights_0=0.0172
    ConstraintWeights_1:float=0.6039
    ConstraintWeights_2:float=0.1324
    ConstraintWeights_3:float=0.2466
   ) -> (Pose) {
    alphaBoolBlend {blendOption=Linear}
    boneToModify {boneName=dyn_belt_strap_MSTR}
    constraintSetup [
     {targetBone {boneName=spine_01}}
     {targetBone {boneName=dyn_belt_buckle_MSTR}}
     {targetBone {boneName=thigh_twist_01_r}}
     {targetBone {boneName=pelvis}}
    ]
    constraintWeights [1,1,1,1]
   }
   n18 = "Constraint - dyn_belt_clip_02_MSTR (PPP)"!(ComponentPose=@n17, Alpha=1, ConstraintWeights_0=0.2816, ConstraintWeights_1:float=0.0275, ConstraintWeights_2:float=0.6909) -> (Pose) {
    alphaBoolBlend {blendOption=Linear}
    boneToModify {boneName=dyn_belt_clip_02_MSTR}
    constraintSetup [{targetBone {boneName=pelvis}}, {targetBone {boneName=thigh_twist_01_l}}, {targetBone {boneName=thigh_twist_01_r}}]
    constraintWeights [1,1,1]
   }
   n19 = "Constraint - dyn_belt_clip_02_Collision (PPPP)"!(
    ComponentPose=@n18
    Alpha=1
    ConstraintWeights_0=0.0137
    ConstraintWeights_1:float=0.048
    ConstraintWeights_2:float=0.0539
    ConstraintWeights_3:float=0.8844
   ) -> (Pose) {
    alphaBoolBlend {blendOption=Linear}
    boneToModify {boneName=dyn_belt_clip_02_Collision}
    constraintSetup [
     {targetBone {boneName=thigh_r}}
     {targetBone {boneName=pelvis}}
     {targetBone {boneName=thigh_twist_01_l}}
     {targetBone {boneName=thigh_twist_01_r}}
    ]
    constraintWeights [1,1,1,1]
   }
   n20 = "Constraint - dyn_shoe_race_r_out_MSTR1 (PP)"!(ComponentPose=@n19, Alpha=1, ConstraintWeights_0=0.6771, ConstraintWeights_1:float=0.3229) -> (Pose) {
    alphaBoolBlend {blendOption=Linear}
    boneToModify {boneName=dyn_shoe_race_r_out_MSTR1}
    constraintSetup [{targetBone {boneName=calf_twist_01_r}}, {targetBone {boneName=foot_r}}]
    constraintWeights [1,1]
   }
   n21 = "Constraint - dyn_shoe_band_r_MSTR (PP)"!(ComponentPose=@n20, Alpha=1, ConstraintWeights_0=0.392, ConstraintWeights_1:float=0.608) -> (Pose) {
    alphaBoolBlend {blendOption=Linear}
    boneToModify {boneName=dyn_shoe_band_r_MSTR}
    constraintSetup [{targetBone {boneName=calf_twist_01_r}}, {targetBone {boneName=foot_r}}]
    constraintWeights [1,1]
   }
   n22 = "Constraint - dyn_shoe_race_r_in_MSTR (PP)"!(ComponentPose=@n21, Alpha=1, ConstraintWeights_0=0.6744, ConstraintWeights_1:float=0.3256) -> (Pose) {
    alphaBoolBlend {blendOption=Linear}
    boneToModify {boneName=dyn_shoe_race_r_in_MSTR}
    constraintSetup [{targetBone {boneName=calf_twist_01_r}}, {targetBone {boneName=foot_r}}]
    constraintWeights [1,1]
   }
   n23 = "Constraint - dyn_shoe_lace_l_in_MSTR (PP)"!(ComponentPose=@n22, Alpha=1, ConstraintWeights_0=0.6744, ConstraintWeights_1:float=0.3256) -> (Pose) {
    alphaBoolBlend {blendOption=Linear}
    boneToModify {boneName=dyn_shoe_lace_l_in_MSTR}
    constraintSetup [{targetBone {boneName=calf_twist_01_l}}, {targetBone {boneName=foot_l}}]
    constraintWeights [1,1]
   }
   n24 = "Constraint - dyn_shoe_band_l_MSTR (PP)"!(ComponentPose=@n23, Alpha=1, ConstraintWeights_0=0.392, ConstraintWeights_1:float=0.608) -> (Pose) {
    alphaBoolBlend {blendOption=Linear}
    boneToModify {boneName=dyn_shoe_band_l_MSTR}
    constraintSetup [{targetBone {boneName=calf_twist_01_l}}, {targetBone {boneName=foot_l}}]
    constraintWeights [1,1]
   }
   n25 = "Constraint - dyn_shoe_lace_l_out_MSTR1 (PP)"!(ComponentPose=@n24, Alpha=1, ConstraintWeights_0=0.6771, ConstraintWeights_1:float=0.3229) -> (Pose) {
    alphaBoolBlend {blendOption=Linear}
    boneToModify {boneName=dyn_shoe_lace_l_out_MSTR1}
    constraintSetup [{targetBone {boneName=calf_twist_01_l}}, {targetBone {boneName=foot_l}}]
    constraintWeights [1,1]
   }
   n26 = "Transform (Modify) Bone - Bone: dyn_elbow_pad_l_MSTR"(
    ComponentPose=@n25
    Alpha=Get elbow_rot_l() -> (elbow_rot_l:double)
    Translation="-0.320,1.183,-0.000"
    Rotation="0.000,20.002,0.000"
    Scale="1,1,1"
   ) {
    rotationMode=BMM_Additive
    rotationSpace=BCS_BoneSpace
    translationMode=BMM_Additive
    translationSpace=BCS_BoneSpace
    alphaBoolBlend {blendOption=Linear}
    alphaScaleBiasClamp {bInterpResult=true}
    boneToModify {boneName=dyn_elbow_pad_l_MSTR}
    rotation {yaw=20.0019}
    translation=<-0.3199 1.1828 0>
   }
   n27 = "Transform (Modify) Bone - Bone: dyn_elbow_pad_r_MSTR"(
    ComponentPose=@n26
    Alpha=Get elbow_rot_r() -> (elbow_rot_r:double)
    Translation="-1.264,-0.959,0.000"
    Rotation="0.000,-10.000,0.000"
    Scale="1,1,1"
   ) {
    rotationMode=BMM_Additive
    rotationSpace=BCS_BoneSpace
    translationMode=BMM_Additive
    translationSpace=BCS_BoneSpace
    alphaBoolBlend {blendOption=Linear}
    alphaScaleBiasClamp {bInterpResult=true}
    boneToModify {boneName=dyn_elbow_pad_r_MSTR}
    rotation {yaw=-10.0003}
    translation=<-1.2639 -0.959 0>
   }
   n28 = "Transform (Modify) Bone - Bone: dyn_knee_pad_MSTR"(
    ComponentPose=@n27
    Alpha=Get knee_rot_r() -> (knee_rot_r:double)
    Translation="-0.927,-0.000,-0.509"
    Rotation="-15.000,0.000,0.000"
    Scale="1,1,1"
   ) {
    rotationMode=BMM_Additive
    rotationSpace=BCS_BoneSpace
    translationMode=BMM_Additive
    translationSpace=BCS_BoneSpace
    alphaBoolBlend {blendOption=Linear}
    alphaScaleBiasClamp {bInterpResult=true}
    boneToModify {boneName=dyn_knee_pad_MSTR}
    rotation {pitch=-15}
    translation {x=-0.927, z=-0.509}
   }
   n29 = RigidBody(ComponentPose=@n28) {
    bClampLinearTranslationLimitToRefPose=true
    bOverrideWorldGravity=true
    overridePhysicsAsset=&/M_MED_Pug_Militia_Dyn_torso
    simulationSpace=BaseBoneSpace
    alphaBoolBlend {blendOption=Linear}
    baseBoneRef {boneName=pelvis}
    componentAppliedLinearAccClamp=<3000 3000 3000>
    componentLinearAccScale=<0.3 0.2 0.2>
    componentLinearVelScale=<0.3 0.2 0.2>
    overrideWorldGravity {z=-980}
   }
   n30 = RigidBody(ComponentPose=@n29) {
    bClampLinearTranslationLimitToRefPose=true
    bOverrideWorldGravity=true
    overridePhysicsAsset=&/M_MED_Pug_Militia_Dyn_chest
    simulationSpace=BaseBoneSpace
    alphaBoolBlend {blendOption=Linear}
    baseBoneRef {boneName=spine_04}
    componentAppliedLinearAccClamp=<4000 4000 4000>
    componentLinearAccScale=<0.4 0.3 0.3>
    componentLinearVelScale=<0.4 0.3 0.3>
    overrideWorldGravity {z=-980}
   }
   n31 = RigidBody!(ComponentPose=@n30, bAlphaBoolEnabled:bool=NOT Boolean(A=Get isWearingBackbling()), ExternalForce) -> (Pose) {
    alphaInputType=Bool
    bClampLinearTranslationLimitToRefPose=true
    bOverrideWorldGravity=true
    overridePhysicsAsset=&/M_MED_Pug_Militia_Dyn_back
    simulationSpace=BaseBoneSpace
    alphaBoolBlend {blendOption=Linear}
    baseBoneRef {boneName=spine_04}
    componentAppliedLinearAccClamp=<4000 4000 4000>
    componentLinearAccScale=<0.4 0.3 0.3>
    componentLinearVelScale=<0.4 0.3 0.3>
    overrideWorldGravity {z=-980}
   }
   n32 = RigidBody!(ComponentPose=@n31, Alpha:float, ExternalForce) -> (Pose) {
    bClampLinearTranslationLimitToRefPose=true
    bOverrideWorldGravity=true
    overridePhysicsAsset=&/M_MED_Pug_Militia_Dyn_leg_l
    simulationSpace=BaseBoneSpace
    alphaBoolBlend {blendOption=Linear}
    baseBoneRef {boneName=thigh_l}
    componentAppliedLinearAccClamp=<3000 3000 3000>
    componentLinearAccScale=<0.3 0.2 0.2>
    componentLinearVelScale=<0.3 0.2 0.2>
    overrideWorldGravity {z=-980}
   }
   n33 = RigidBody(ComponentPose=@n32) {
    bClampLinearTranslationLimitToRefPose=true
    bOverrideWorldGravity=true
    overridePhysicsAsset=&/M_MED_Pug_Militia_Dyn_leg_r
    simulationSpace=BaseBoneSpace
    alphaBoolBlend {blendOption=Linear}
    baseBoneRef {boneName=thigh_r}
    componentAppliedLinearAccClamp=<3000 3000 3000>
    componentLinearAccScale=<0.3 0.2 0.2>
    componentLinearVelScale=<0.3 0.2 0.2>
    overrideWorldGravity {z=-980}
   }
   n34 = RigidBody(ComponentPose=@n33) {
    bClampLinearTranslationLimitToRefPose=true
    bOverrideWorldGravity=true
    overridePhysicsAsset=&/M_MED_Pug_Militia_Dyn_belly
    simulationSpace=BaseBoneSpace
    alphaBoolBlend {blendOption=Linear}
    baseBoneRef {boneName=spine_01}
    componentAppliedLinearAccClamp=<3000 3000 3000>
    componentLinearAccScale=<0.3 0.2 0.2>
    componentLinearVelScale=<0.3 0.2 0.2>
    overrideWorldGravity {z=-980}
   }
   Output Pose(Result=Component To Local(ComponentPose=@n34)) {layerGroup=None}
  ]}
 EventGraph {nodes [
   Event BlueprintUpdateAnimation()
   n1 = Get Skel_Mesh()
   > Set knee_rot_r(knee_rot_r:double=GetAlphaFromReferenceTransformDelta(SkeletalMeshComponent=@n1, BoneName=calf_r, BaseName=thigh_r, InputRange=(X=0,Y=90), TwistAxis="0,0,1")) -> (Output_Get:double)
   > Set elbow_rot_l(elbow_rot_l:double=GetAlphaFromReferenceTransformDelta(SkeletalMeshComponent=@n1, BoneName=lowerarm_l, BaseName=upperarm_l, InputRange=(X=0,Y=60), TwistAxis="0,0,1")) -> (Output_Get:double)
   > Set elbow_rot_r(elbow_rot_r:double=GetAlphaFromReferenceTransformDelta(SkeletalMeshComponent=@n1, BoneName=lowerarm_r, BaseName=upperarm_r, InputRange=(X=0,Y=60), TwistAxis="0,0,1")) -> (Output_Get:double)
   Event OnCharacterCustomizationLoaded()
   > Do Once()
   n2 = CheckForBackbling()
   > Set isWearingBackbling(isWearingBackbling=OR Boolean(A=@n2, B=@n2.bIsBackblingLong))
   Event BlueprintInitializeAnimation()
   > Set Skel_Mesh(Skel_Mesh=GetOwnerSkeletalMeshComponent())
  ]}
]
variables [
 elbow_rot_l {type=number}
 elbow_rot_r {type=number}
 isWearingBackbling {type=boolean}
 knee_rot_r {type=number}
 Skel_Mesh {type=SkeletalMeshComponent}
]
