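# Physics asset for the face accessory mesh of the male medium Pug Militia character body, set up for a dynamic necklace.
# Two capsule bodies on the necklace bones form a chain from spine_05, with tight limits at the root joint and wider swing at the second link.
# Capsules on spine_04 and spine_05, four of them on the latter, approximate the upper torso, and the constraint between those two spine bones leaves swing and twist free.
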PhysicsAsset /BRCosmetics/Characters/Player/Male/Medium/Bodies/M_MED_Pug_Militia/Meshes/Parts/M_MED_Pug_Militia_FaceAcc_Dyn
bodies {
 dyn_necklace_01 [{shapeType=Capsule, length=5.0625, radius=2.125, center=<4.1774 -0.8169 0>, rotation=<rot 90 -90 -79.9997>}]
 dyn_necklace_02 [{shapeType=Capsule, length=5.5938, radius=2, center=<4.1396 -1.2002 0>, rotation=<rot 90 0 5.0003>}]
 spine_04 [{shapeType=Capsule, length=5, radius=5}]
 spine_05 [
  {shapeType=Capsule, length=20.5938, radius=13.6562, center=<-0.236 -2.2751 0>}
  {shapeType=Capsule, length=15, radius=15, center=<-6.6186 2.431 0>}
  {shapeType=Capsule, length=15, radius=4.5938, center=<6.5362 10.0912 -9.8765>, rotation=<rot -16.1398 51.0414 -108.345>}
  {shapeType=Capsule, length=10.625, radius=4.5625, center=<-1.7346 18.7625 8.5886>, rotation=<rot -81.2338 177.8506 172.2678>}
 ]
}
constraints [
 {bone1Name=dyn_necklace_01, bone2Name=spine_05, swing1LimitDegrees=10, swing2LimitDegrees=10, twistLimitDegrees=5}
 {bone1Name=dyn_necklace_02, bone2Name=dyn_necklace_01, swing1LimitDegrees=45, swing2LimitDegrees=45, twistLimitDegrees=5}
 {
  bone1Name=spine_05
  bone2Name=spine_04
  swing1LimitDegrees=45
  swing1Motion=Free
  swing2LimitDegrees=45
  swing2Motion=Free
  twistLimitDegrees=45
  twistMotion=Free
 }
]
